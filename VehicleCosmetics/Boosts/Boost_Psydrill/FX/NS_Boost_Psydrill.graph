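# Niagara effect for the Psydrill car boost cosmetic: a wobbling drill cone with backward-streaming particles and trailing ribbons.
# The cone merges burst-spawned mesh and sprite layers in one emitter, with spawn groups setting scale, offset, sine wobble and render tag.
# Rate-spawned sprites and meshes fly backward, while quality-scaled ribbons inherit source movement.
# Paint-variant overrides color every emitter, and painted and frontend-preview flags adjust color and motion.
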NiagaraSystem /VehicleCosmetics/Boosts/Boost_Psydrill/FX/NS_Boost_Psydrill
dependencies {
 usedDataInterfaces [
  &NiagaraDataInterfaceCurve
  &NiagaraDataInterfaceVectorCurve
  &NiagaraDataInterfaceColorCurve
  &NiagaraDataInterfaceArrayColor
  &NiagaraDataInterfaceArrayFloat3
  &NiagaraDataInterfaceSkeletalMesh
 ]
 usedDynamicInputs []
 usedModules [
  &self:SystemSpawnScript.SystemScriptSource.SystemScriptGraph.NiagaraNodeAssignment_0.SetVariables_#1
  &SystemState
 ]
}
emitters {
 NE_Psydrill_Cone001 {
  emitterUpdateScript [
   &EmitterState(
    Inactive Response=::NewEnumerator1
    Life Cycle Mode=::NewEnumerator1
    Loop Behavior=::NewEnumerator1
    Loop Duration Mode=::NewEnumerator1
    UseLoopDelay=false
   )
   &VC_FilteredSocketCount()
   &SpawnRate(SpawnRate=10, Use Spawn Probability=false) {enabled=false}
   &SpawnBurst_Instantaneous(Use Spawn Probability=false, UseLoopCountLimit=true)
   &SpawnBurst_Instantaneous(Spawn Group=1, Use Spawn Probability=false, UseLoopCountLimit=true) {module=SpawnBurst_Instantaneous002}
   &SpawnBurst_Instantaneous(Spawn Group=2, Use Spawn Probability=false, UseLoopCountLimit=true) {module=SpawnBurst_Instantaneous001}
   &SpawnBurst_Instantaneous(Spawn Group=3, Use Spawn Probability=false, UseLoopCountLimit=false) {module=SpawnBurst_Instantaneous003}
   &SpawnRate(SpawnGroup=3, SpawnRate=13, Use Spawn Probability=false) {enabled=false, module=SpawnRate001}
  ]
  particleSpawnScript [
   &VC_PaintedVariantsOverride(
    Black=<rgba 0.01 0.01 0.01 1>
    BurntSienna=<rgba 0.09 0.015 0 1>
    Cobalt=<rgba 0 0.0492 0.8 1>
    Crimson=<rgba 0.4635 0.0093 0.0216 1>
    DefaultColor=<rgba 0.3 0.03 0.0525 1>
    ForestGreen=<rgba 0.084 0.4375 0.041 1>
    Lime=<rgba 0.3249 0.5885 0.0613 1>
    Orange=<rgba 0.6724 0.0563 0.0003 1>
    Pink=<rgba 0.5781 0.1355 0.4172 1>
    Purple=<rgba 0.2358 0.0781 1 1>
    R1 Black:LinearColor=<rgba 0.15 0.15 0.15 1>
    R1 Cobalt:LinearColor=<rgba 0.0092 0.0146 0.5677 1>
    R1 Crimson:LinearColor=<rgba 5 0 0.1355 1>
    R1 ForestGreen:LinearColor=<rgba 0.259 0.5 0.2396 1>
    R1 Lime:LinearColor=<rgba 1.1488 2 0.3601 1>
    R1 Orange:LinearColor=<rgba 0.6724 0.0672 0 1>
    R1 Pink:LinearColor=<rgba 1 0 0.9441 0>
    R1 Saffron:LinearColor=<rgba 10 7.6667 0 1>
    R1 SkyBlue:LinearColor=<rgba 0.5365 0.5365 0.5365 1>
    R1 TitaniumWhite:LinearColor=<rgba 0.2604 0.2604 0.2604 1>
    R2 Black:LinearColor=<rgba 0.025 0.025 0.025 400>
    R2 BurntSienna:LinearColor=<rgba 1.1 0.2327 0 1>
    R2 Cobalt:LinearColor=<rgba 0.0678 0.1837 1 100>
    R2 ForestGreen:LinearColor=<rgba 0.05 2 0 10>
    R2 Grey:LinearColor=<rgba 0.1 0.1 0.1 5>
    R2 Lime:LinearColor=<rgba 0.5 1 0 5>
    R2 Orange:LinearColor=<rgba 1 0.3488 0 1.5>
    R2 Pink:LinearColor=<rgba 1 0.25 0.7273 2>
    R2 Purple:LinearColor=<rgba 1.2312 0 3 1>
    R2 Saffron:LinearColor=<rgba 1.5 1.15 0 1>
    R2 SkyBlue:LinearColor=<rgba 0.0002 3084.8896 3500 1>
    R2 TitaniumWhite:LinearColor=<rgba 1 1 1 1>
    R3 Black:LinearColor=<rgba 0.002 0.002 0.002 20>
    R3 BurntSienna:LinearColor=<rgba 0.03 0.005 0 1>
    R3 Cobalt:LinearColor=<rgba 0 0.0174 0.5 1>
    R3 Crimson:LinearColor=<rgba 0.4635 0 0.0126 1>
    R3 DefaultColor:LinearColor=<rgba 0.85 0 0.0567 5>
    R3 Grey:LinearColor=<rgba 0.099 0.099 0.099 1>
    R3 Lime:LinearColor=<rgba 0.1766 0.4 0.02 1>
    R3 Orange:LinearColor=<rgba 0.6198 0.0516 0 1>
    R3 Pink:LinearColor=<rgba 0.474 0.0346 0.3588 1>
    R3 Purple:LinearColor=<rgba 0.1697 0.0389 0.5729 1>
    R3 Saffron:LinearColor=<rgba 0.2706 0.2471 0.0078 1>
    R3 SkyBlue:LinearColor=<rgba 0.0268 0.2816 0.5365 1>
    R3 TitaniumWhite:LinearColor=<rgba 0.2448 0.2448 0.2448 1>
    R4 Black:LinearColor=<rgba 0 0 0 1>
    R4 BurntSienna:LinearColor=<rgba 0.08 0.0133 0 2>
    R4 Cobalt:LinearColor=<rgba 0 0.0704 0.5 1>
    R4 Crimson:LinearColor=<rgba 1.5 0 0.0407 1>
    R4 DefaultColor:LinearColor=<rgba 1 0.25 0.25 0.5>
    R4 Grey:LinearColor=<rgba 0.05 0.05 0.05 1.5>
    R4 Orange:LinearColor=<rgba 0.45 0.03 0 1.5>
    R4 Pink:LinearColor=<rgba 0.3 0.0359 0.204 1>
    R4 Purple:LinearColor=<rgba 0.1699 0.0308 0.3698 1>
    R4 Saffron:LinearColor=<rgba 0.3 0.275 0 1>
    R4 TitaniumWhite:LinearColor=<rgba 0.3 0.3 0.3 1>
    Saffron=<rgba 0.4 0.3589 0.0625 1>
    SkyBlue=<rgba 0.0698 0.4288 0.6094 1>
    TitaniumWhite=<rgba 0.4792 0.4792 0.4792 1>
    Use R1 Painted Overrides=true
    Use R2 Painted Overrides=true
    Use R3 Painted Overrides=true
    Use R4 Painted Overrides=true
   )
   /SetVariables(Particles.RenderVisTag:NiagaraInt32=&SpawnGroup()) {isSetParameters=true}
   &InitializeParticle(
    Color Mode=::NewEnumerator0
    Lifetime=/VehicleCosmetics/FX/Niagara/Modules/MergeEmitters/VC_FloatBySpawnGroup(
     SpawnGroup or ExecIndex:NiagaraInt32=0
     CustomInt:NiagaraInt32=0
     MaximumSpawnGroups:NiagaraInt32=5
     Float SG 0:NiagaraFloat=$Emitter.CurrentLoopDuration
     Float SG 1:NiagaraFloat=$Emitter.CurrentLoopDuration
     Float SG 2:NiagaraFloat=$Emitter.CurrentLoopDuration
     Float SG 3:NiagaraFloat=0.15
     Float SG 4:NiagaraFloat=0.1
     Float SG 5:NiagaraFloat=0
    )
    Mesh Scale=&Multiply_VectorByFloat(
     Vector=/VehicleCosmetics/FX/Niagara/Modules/MergeEmitters/VC_VectorBySpawnGroup(
      SpawnGroup or ExecIndex:NiagaraInt32=0
      CustomInt:NiagaraInt32=0
      MaximumSpawnGroups:NiagaraInt32=4
      Vector Group 0:Vector3f=<0.48 0.28 0.28>
      Vector Group 1:Vector3f=<0.48 0.28 0.28>
      Vector Group 2:Vector3f=<0.6 0.5 0.5>
      Vector Group 3:Vector3f=<0.52 0.11 0.11>
      Vector Group 4:Vector3f=<0 0 0>
      Vector Group 5:Vector3f=<0 0 0>
     )
     Float=1.1
    )
    Mesh Scale Mode=::NewEnumerator4
    Position Offset=&AddVectorToPosition(
     Position=$System.FinalPosition
     Vector=&VC_VectorBySpawnGroup(
      SpawnGroup or ExecIndex:NiagaraInt32=0
      CustomInt:NiagaraInt32=0
      MaximumSpawnGroups:NiagaraInt32=5
      Vector Group 0:Vector3f=<-70 0 0>
      Vector Group 1:Vector3f=<-70 0 0>
      Vector Group 2:Vector3f=<-68 0 0>
      Vector Group 3:Vector3f=&RandomRangeVector(Minimum=<-25 0 0>, Maximum=<-25 0 0>)
      Vector Group 4:Vector3f=<-80 0 0>
      Vector Group 5:Vector3f=<0 0 0>
     )
    )
    Sprite Size=<100 85>
    Sprite Size Mode=::NewEnumerator4
    UsePositionOffset=true
    Write Material Random=true
   )
   &PartitionParticles(Number Of Partitions=$Emitter.PortCount) {enabled=false}
   &SkeletalMeshLocation(
    Bone / Socket Index=$Particles.PartitionParticles.Partition
    Bone / Socket Sampling Mode=::NewEnumerator5
    MaskSpawnGroup=false
    Mesh Sampling Type=::NewEnumerator6
    Skeletal Mesh=$System.SkeletalMesh
   ) {enabled=false}
   &ShapeLocation(
    MaskSpawnGroup=false
    Offset=&MakeVector(X=&Sine(Normalized Angle=$Emitter.Age, Period=1, Scale=50, Bias=0), Y=0, Z=0)
    Offset Mode=::NewEnumerator0
    Override Seed=false
    Sphere Radius=0
    UseOwnerScale=true
   ) {enabled=false}
   &AddVelocity(Velocity=<-1500 0 0>) {enabled=false}
   &InitialMeshOrientation(
    Mesh Orientation Mode=::NewEnumerator2
    Orientation Vector=<-1 0 0>
    Rotation=&VC_VectorBySpawnGroup(
     SpawnGroup or ExecIndex:NiagaraInt32=0
     CustomInt:NiagaraInt32=0
     MaximumSpawnGroups:NiagaraInt32=5
     Vector Group 0:Vector3f=<0 0 0>
     Vector Group 1:Vector3f=<0 0 0>
     Vector Group 2:Vector3f=<0 0 0>
     Vector Group 3:Vector3f=&RandomRangeVector(Minimum=<0 0 0>, Maximum=<1 0 0>)
     Vector Group 4:Vector3f=<0 0 0>
     Vector Group 5:Vector3f=<0 0 0>
    )
    Use Rotation Vector=true
   )
  ]
  particleUpdateScript [
   &ParticleState(Kill Particles When Lifetime Has Elapsed=false, Loop Particles Lifetime=true)
   &ScaleSpriteSize(Uniform Curve Sprite Scale=DataInterface() {propertyValues {Curve {defaultValue=1, keys [{tangentMode=RCTM_User, time=0, value=0.5}, {tangentMode=RCTM_User, time=1, value=1}]}}})
   &ScaleMeshSize(Scale Factor=&VC_VectorBySpawnGroup(
     SpawnGroup or ExecIndex:NiagaraInt32=0
     CustomInt:NiagaraInt32=0
     MaximumSpawnGroups:NiagaraInt32=5
     Vector Group 0:Vector3f=<1 1 1>
     Vector Group 1:Vector3f=<1 1 1>
     Vector Group 2:Vector3f=<1 1 1>
     Vector Group 3:Vector3f=&VectorFromCurve(
      VectorCurve=DataInterface() {propertyValues {XCurve {keys [<at 0 0.5>,<at 1 4.5>]}, YCurve {keys [<at 0 0.9>,<at 1 2.5>]}, ZCurve {keys [<at 0 0.9>,<at 1 2.5>]}}}
      Scale Curve=<0.8 1 1>
     )
     Vector Group 4:Vector3f=<1 1 1>
     Vector Group 5:Vector3f=<0 0 0>
    ))
   &SolveForcesAndVelocity(Clamp Velocity=false, Limit Acceleration=false)
   &ScaleColor(
    Scale Alpha=0.75
    Scale RGB=&MakeCustomVectorFromBool(Bool=$User.bIsPainted, True Vector=<1 1 1>, False Vector=<5 2 2>)
    ScaleA=true
    ScaleRGB=true
   )
   &DynamicMaterialParameters(
    Index 2 Param 1=1
    Index 2 Param 2=1
    Index 2 Param 3=1
    Index 2 Param 4=1
    Module.Index2Param0Write=true
    Module.Index2Param1Write=true
    Module.Index2Param2Write=true
    Module.Index2Param3Write=true
    Write Parameter Index 0=false
    Write Parameter Index 2=true
   ) {enabled=false}
   /SetVariables(
    Particles.R1 Scaled Color:LinearColor=&VC_DynamicScaleColor(Color to Scale:LinearColor=$Particles.R1 Color, RGB Scale:Vector3f=<1 1 1>, Alpha Scale:NiagaraFloat=1)
    Particles.R2 Scaled Color:LinearColor=&VC_DynamicScaleColor(Color to Scale:LinearColor=$Particles.R2 Color, RGB Scale:Vector3f=<1 1 1>, Alpha Scale:NiagaraFloat=5)
    Particles.R3 Scaled Color:LinearColor=&VC_DynamicScaleColor(Color to Scale:LinearColor=$Particles.R3 Color, RGB Scale:Vector3f=<2.5 2.5 2.5>, Alpha Scale:NiagaraFloat=1)
    Particles.R4 Scaled Color:LinearColor=&VC_DynamicScaleColor(Color to Scale:LinearColor=$Particles.R4 Color, RGB Scale:Vector3f=<5 5 5>, Alpha Scale:NiagaraFloat=0.2)
   ) {isSetParameters=true}
   &ShapeLocation(
    MaskSpawnGroup=false
    Offset=&Multiply_VectorByFloat(
     Vector=&MakeVector(
      X=&Sine(Normalized Angle=$Emitter.Age, Period=0.45, Scale=0.45, Bias=0.65)
      Y=&Sine(
       Normalized Angle=$Emitter.Age
       Period=&VC_FloatBySpawnGroup(
        SpawnGroup or ExecIndex:NiagaraInt32=0
        CustomInt:NiagaraInt32=0
        MaximumSpawnGroups:NiagaraInt32=3
        Float SG 0:NiagaraFloat=0.75
        Float SG 1:NiagaraFloat=0.75
        Float SG 2:NiagaraFloat=1
        Float SG 3:NiagaraFloat=0
        Float SG 4:NiagaraFloat=0
        Float SG 5:NiagaraFloat=0
       )
       Scale=1.5
       Bias=0.1
      )
      Z=&Sine(
       Normalized Angle=$Emitter.Age
       Period=&VC_FloatBySpawnGroup(
        SpawnGroup or ExecIndex:NiagaraInt32=0
        CustomInt:NiagaraInt32=0
        MaximumSpawnGroups:NiagaraInt32=3
        Float SG 0:NiagaraFloat=0.45
        Float SG 1:NiagaraFloat=0.45
        Float SG 2:NiagaraFloat=2
        Float SG 3:NiagaraFloat=0
        Float SG 4:NiagaraFloat=0
        Float SG 5:NiagaraFloat=0
       )
       Scale=1
       Bias=&VC_FloatBySpawnGroup(
        SpawnGroup or ExecIndex:NiagaraInt32=0
        CustomInt:NiagaraInt32=0
        MaximumSpawnGroups:NiagaraInt32=3
        Float SG 0:NiagaraFloat=0.7
        Float SG 1:NiagaraFloat=0.7
        Float SG 2:NiagaraFloat=0.2
        Float SG 3:NiagaraFloat=0
        Float SG 4:NiagaraFloat=0
        Float SG 5:NiagaraFloat=0
       )
      )
     )
     Float=&VC_FloatBySpawnGroup(
      SpawnGroup or ExecIndex:NiagaraInt32=0
      CustomInt:NiagaraInt32=0
      MaximumSpawnGroups:NiagaraInt32=5
      Float SG 0:NiagaraFloat=1
      Float SG 1:NiagaraFloat=1
      Float SG 2:NiagaraFloat=1
      Float SG 3:NiagaraFloat=0
      Float SG 4:NiagaraFloat=0
      Float SG 5:NiagaraFloat=0
     )
    )
    Offset Mode=::NewEnumerator0
    Override Seed=false
    Shape Origin=$Particles.Initial.Position
    Sphere Radius=0
    UseOwnerScale=true
   ) {module=ShapeLocation001}
  ]
  rendererClasses [&NiagaraMeshRendererProperties, &NiagaraSpriteRendererProperties]
  renderers [
   {rendererClass=&NiagaraMeshRendererProperties, rendererIndex=0}
   {rendererClass=&NiagaraMeshRendererProperties, rendererIndex=1}
   {rendererClass=&NiagaraMeshRendererProperties, rendererIndex=2}
   {rendererClass=&NiagaraMeshRendererProperties, rendererIndex=3}
   {rendererClass=&NiagaraSpriteRendererProperties, rendererIndex=4}
  ]
 }
 NE_Psydrill_Particles001 {
  emitterUpdateScript [
   &EmitterState()
   &SpawnRate(SpawnRate=80, Use Spawn Probability=false) {module=SpawnRate001}
   &SpawnRate(SpawnGroup=1, SpawnRate=15, Use Spawn Probability=false)
  ]
  particleSpawnScript [
   &VC_PaintedVariantsOverride(
    Black=<rgba 0.002 0.002 0.002 1.5>
    BurntSienna=<rgba 0.05 0.0154 0.0062 1>
    Crimson=<rgba 0.4635 0 0.0126 1>
    DefaultColor=&SelectLinearColorFromArray(
     Array Sampling Mode=::NewEnumerator0
     Direct Array Index=?
     Color Selection Array=DataInterface() {propertyValues {ColorData [<rgba 1 0.1 1 1>, <rgba 1 0.05 0.1 1>]}}
     Random Seed=0
    )
    R1 Black:LinearColor=<rgba 0.002 0.002 0.002 2>
    R1 BurntSienna:LinearColor=<rgba 0.03 0.004 0 5>
    R1 Cobalt:LinearColor=<rgba 0 0.0012 0.5 1>
    R1 Crimson:LinearColor=<rgba 0.2 0 0.0054 1>
    R1 DefaultColor=<rgba 0.9 0 0.0524 1>
    R1 ForestGreen:LinearColor=<rgba 0 0.3125 0.0158 1>
    R1 Grey:LinearColor=<rgba 0.05 0.05 0.05 1>
    R1 Lime:LinearColor=<rgba 0.0883 0.2 0.01 1>
    R1 Orange:LinearColor=<rgba 0.25 0.0208 0 1>
    R1 Pink:LinearColor=<rgba 0.25 0.0299 0.173 1>
    R1 Purple:LinearColor=<rgba 0.0733 0.0226 0.2396 1>
    R1 Saffron:LinearColor=<rgba 0.1 0.0918 0.008 1>
    R1 SkyBlue:LinearColor=<rgba 0 0.1331 0.2 1>
    R1 TitaniumWhite:LinearColor=<rgba 0.1 0.1 0.1 1>
    Use R1 Painted Overrides=true
   )
   &InitializeParticle(
    Color Mode=::NewEnumerator0
    Lifetime=&VC_FloatBySpawnGroup(
     SpawnGroup or ExecIndex:NiagaraInt32=0
     CustomInt:NiagaraInt32=0
     MaximumSpawnGroups:NiagaraInt32=2
     Float SG 0:NiagaraFloat=&RandomRangeFloat(Minimum=0.4, Maximum=0.5)
     Float SG 1:NiagaraFloat=&MakeCustomFloatFromBool(Bool=$User.bIsFrontendPreview, True Float=0.3, False Float=0.3)
     Float SG 2:NiagaraFloat=0
     Float SG 3:NiagaraFloat=0
     Float SG 4:NiagaraFloat=0
     Float SG 5:NiagaraFloat=0
    )
    Mesh Scale=<0.48 0.12 0.12>
    Mesh Scale Mode=::NewEnumerator4
    Position Offset=&AddVectorToPosition(
     Position=$System.FinalPosition
     Vector=&VC_VectorBySpawnGroup(
      SpawnGroup or ExecIndex:NiagaraInt32=0
      CustomInt:NiagaraInt32=0
      MaximumSpawnGroups:NiagaraInt32=2
      Vector Group 0:Vector3f=<-110 0 0>
      Vector Group 1:Vector3f=&SelectVectorFromArray(
       Array Sampling Mode=::NewEnumerator0
       Direct Array Index=?
       Vector Selection Array=DataInterface() {propertyValues {FloatData [<-70 0 0>, <-90 0 0>]}}
       Random Seed=0
      )
      Vector Group 2:Vector3f=<0 0 0>
      Vector Group 3:Vector3f=<0 0 0>
      Vector Group 4:Vector3f=<0 0 0>
      Vector Group 5:Vector3f=<0 0 0>
     )
    )
    Sprite Rotation Mode=::NewEnumerator2
    Sprite Size Mode=::NewEnumerator3
    Uniform Sprite Size=8
    UsePositionOffset=true
    Write Material Random=true
   )
   /SetVariables(Particles.RenderVisTag:NiagaraInt32=&SpawnGroup()) {isSetParameters=true}
   &AddVelocity(
    Velocity=&VC_VectorBySpawnGroup(
     SpawnGroup or ExecIndex:NiagaraInt32=0
     CustomInt:NiagaraInt32=0
     MaximumSpawnGroups:NiagaraInt32=2
     Vector Group 0:Vector3f=<-1000 0 0>
     Vector Group 1:Vector3f=<-1500 0 0>
     Vector Group 2:Vector3f=<0 0 0>
     Vector Group 3:Vector3f=<0 0 0>
     Vector Group 4:Vector3f=<0 0 0>
     Vector Group 5:Vector3f=<0 0 0>
    )
    Velocity Speed Scale=&MakeCustomFloatFromBool(Bool=$User.bIsFrontendPreview, True Float=1, False Float=1.5)
   )
   &ShapeLocation(
    Cylinder Height=1
    Cylinder Radius=&VC_FloatBySpawnGroup(
     SpawnGroup or ExecIndex:NiagaraInt32=0
     CustomInt:NiagaraInt32=0
     MaximumSpawnGroups:NiagaraInt32=2
     Float SG 0:NiagaraFloat=30
     Float SG 1:NiagaraFloat=0
     Float SG 2:NiagaraFloat=0
     Float SG 3:NiagaraFloat=0
     Float SG 4:NiagaraFloat=0
     Float SG 5:NiagaraFloat=0
    )
    MaskSpawnGroup=false
    Override Seed=false
    Rotation Angle Type:ENiagara_AngleInput=::NewEnumerator0
    Rotation Mode=::NewEnumerator2
    Shape Primitive=::NewEnumerator1
    Surface Only=false
    UseOwnerScale=true
    Yaw / Pitch / Roll=<0 90 0>
   )
   &InitialMeshOrientation(
    Mesh Orientation Mode=::NewEnumerator2
    Rotation=&RandomRangeVector(Minimum=<0 0 0>, Maximum=<1 0 0>)
    Use Rotation Vector=true
   )
  ]
  particleUpdateScript [
   &ParticleState()
   &InheritSourceMovement(
    Applied Position Scale=&Multiply_VectorByFloat(
     Vector=<1 1 1>
     Float=&MakeCustomFloatFromBool(
      Bool=$User.bIsFrontendPreview
      True Float=1
      False Float=&FloatFromCurve(FloatCurve=DataInterface() {propertyValues {Curve {keys [
           {arriveTangent=0.0579, interpMode=RCIM_Cubic, leaveTangent=0.0579, tangentMode=RCTM_User, time=0.05, value=1}
           {arriveTangent=-2.6543, interpMode=RCIM_Cubic, leaveTangent=-2.6543, tangentMode=RCTM_User, time=1, value=-1}
          ]}}})
     )
    )
    ApplyForce=false
    ApplyPosition=true
    ApplyPositionFalloff=false
   ) {enabled=false}
   &SolveForcesAndVelocity(Clamp Velocity=false, Limit Acceleration=false)
   &ScaleColor(
    Scale RGB=&VectorFromFloat(Value=&RandomRangeFloat(Minimum=2, Maximum=7))
    ScaleA=true
    ScaleRGB=true
   )
   &ScaleMeshSize(Scale Factor=&VectorFromCurve(
     VectorCurve=DataInterface() {propertyValues {XCurve {keys [<at 0 2.2 RCIM_Cubic>,<at 1 1.8 RCIM_Cubic>]}, YCurve {keys [<at 0 1.8>,<at 1 1.2>]}, ZCurve {keys [<at 0 1.8>,<at 1 1.2>]}}}
     Scale Curve=&VectorFromFloat(Value=&FloatFromCurve(FloatCurve=DataInterface() {propertyValues {Curve {keys [<at 0 1>,<at 1 0.9>]}}}))
    ))
   /SetVariables(Particles.R1 Scaled Color:LinearColor=&VC_DynamicScaleColor(
     Color to Scale:LinearColor=$Particles.R1 Color
     RGB Scale:Vector3f=&VectorFromFloat(Value=&FloatFromCurve(
       FloatCurve=DataInterface() {propertyValues {Curve {defaultValue=0, keys [{tangentMode=RCTM_User, time=0, value=1}, {tangentMode=RCTM_User, time=1, value=0.5}]}}}
       Scale Curve=4
      ))
     Alpha Scale:NiagaraFloat=&FloatFromCurve(FloatCurve=DataInterface() {propertyValues {Curve {keys [<at 0 0.5>,<at 0.8 0.3>,<at 1 0>]}}})
    )) {isSetParameters=true}
   &DynamicMaterialParameters(
    Index 1 Param 1=&FloatFromCurve(
     FloatCurve=DataInterface() {propertyValues {Curve {keys [
         {arriveTangent=-0.0405, interpMode=RCIM_Cubic, leaveTangent=-0.0405, tangentMode=RCTM_User, time=0, value=1}
         {arriveTangent=-2.1747, interpMode=RCIM_Cubic, leaveTangent=-2.1747, tangentMode=RCTM_User, time=1, value=0}
        ]}}}
     Scale Curve=&MakeCustomFloatFromBool(Bool=$User.bIsFrontendPreview, True Float=1, False Float=2)
    )
    Index 1 Param 2=&FloatFromCurve(
     FloatCurve=DataInterface() {propertyValues {Curve {keys [
         {arriveTangent=-1.1485, interpMode=RCIM_Cubic, leaveTangent=-1.1485, tangentMode=RCTM_User, time=0, value=0.5}
         {interpMode=RCIM_Cubic, leaveTangent=-0.0796, time=1, value=0}
        ]}}}
     Scale Curve=0.8
    )
    Module.Index0Float0Write=true
    Module.Index0Float1Write=true
    Module.Index0Float2Write=true
    Module.Index0Float3Write=true
    Module.Index1Float0Write=true
    Module.Index1Float1Write=true
    Module.Index1Float2Write=true
    Module.Index1Float3Write=true
    Write Parameter Index 1=::NewEnumerator1
   )
  ]
  rendererClasses [&NiagaraSpriteRendererProperties, &NiagaraMeshRendererProperties]
  renderers [
   {rendererClass=&NiagaraSpriteRendererProperties}
   {rendererClass=&NiagaraMeshRendererProperties, rendererIndex=1}
  ]
 }
 NE_Psydrill_Ribbons001 {
  emitterUpdateScript [
   &EmitterState(Inactive Response=::NewEnumerator1, Life Cycle Mode=::NewEnumerator1, Loop Duration=0.1, UseLoopDelay=false)
   &VC_FilteredSocketCount()
   &VC_RateQualityHandler(High Spawn Rate=60, Low Spawn Rate=30, Medium Spawn Rate=45, Use Custom Rate?=true)
   &VC_SpawnRibbon_v1(SpawnRate=$Emitter.Spawn Rate)
  ]
  particleSpawnScript [
   &VC_PaintedVariantsOverride(
    Black=<rgba 0.002 0.002 0.002 1.5>
    BurntSienna=<rgba 0.1 0.0151 0.002 1.5>
    Cobalt=<rgba 0 0 2 1>
    Crimson=<rgba 1 0.1347 0.12 1>
    DefaultColor=<rgba 1 0.4 0.6 1>
    ForestGreen=<rgba 0 0.35 0.0177 1>
    Grey=<rgba 0.2 0.2 0.2 1>
    Lime=<rgba 0.25 0.5 0 1>
    Orange=<rgba 1 0.2103 0.0129 1>
    Pink=<rgba 1.4505 0.2991 1.5 1>
    Purple=<rgba 0.4333 0 2 1>
    R1 Black:LinearColor=<rgba 1.5 1.5 1.5 0.05>
    R1 BurntSienna:LinearColor=<rgba 0.1667 0.0395 0.02 1>
    R1 Cobalt:LinearColor=<rgba 0.2007 0.3803 1 1>
    R1 Crimson:LinearColor=<rgba 0.8 0.292 0.3058 1>
    R1 DefaultColor=<rgba 50 7.5 18.125 1>
    R1 ForestGreen:LinearColor=<rgba 0.15 1 0.1929 1>
    R1 Grey:LinearColor=<rgba 1 1 1 2>
    R1 Lime:LinearColor=<rgba 1.4255 2 0.2396 1>
    R1 Orange:LinearColor=<rgba 1 0.3 0.125 1>
    R1 Pink:LinearColor=<rgba 2 0.3333 1.3939 1>
    R1 Purple:LinearColor=<rgba 0.4333 0 1.3 1>
    R1 Saffron:LinearColor=<rgba 1 0.9306 0.1667 1>
    R1 SkyBlue:LinearColor=<rgba 0 1.9962 3 1>
    R1 TitaniumWhite:LinearColor=<rgba 1 1 1 1>
    Saffron=<rgba 0.4583 0.5 0 0.75>
    SkyBlue=<rgba 0 0.65 1.5 1>
    TitaniumWhite=<rgba 0.8 0.8 0.8 0.75>
    Use R1 Painted Overrides=true
   )
   &VC_InitializeRibbon_v1(
    Color=$Particles.Color
    Custom Offset=true
    Local Vector Offset=&AddVectorToPosition(
     Position=$System.FinalPosition
     Vector=/Niagara/DynamicInputs/TypeConversions/MakeCustomPositionFromBool(Bool:NiagaraBool=$User.bIsPainted, True Position:NiagaraPosition=<0 0 0>, False Position:NiagaraPosition=<-20 0 0>)
    )
    Offset Amount=1
    Ribbon Width=85
   )
   /SetVariables(Particles.R1 Width=150) {isSetParameters=true}
   &AddVelocity(Velocity=<3500 0 0>, Velocity Speed Scale=-1) {module=AddVelocity001}
  ]
  particleUpdateScript [
   &ParticleState()
   &InheritSourceMovement(
    Applied Position Scale=&Multiply_VectorByFloat(
     Vector=<1 1 1>
     Float=&MakeCustomFloatFromBool(
      Bool=$User.bIsFrontendPreview
      True Float=1
      False Float=&FloatFromCurve(FloatCurve=DataInterface() {propertyValues {Curve {keys [
           {arriveTangent=0.0579, interpMode=RCIM_Cubic, leaveTangent=0.0579, tangentMode=RCTM_User, time=0.05, value=1}
           {arriveTangent=-2.6543, interpMode=RCIM_Cubic, leaveTangent=-2.6543, tangentMode=RCTM_User, time=1, value=-1}
          ]}}})
     )
    )
    ApplyForce=false
    ApplyPosition=true
    ApplyPositionFalloff=false
   )
   &SolveForcesAndVelocity(Clamp Velocity=false, Limit Acceleration=false)
   &VC_UpdateRibbon_v1()
   &ScaleColor(
    Scale Alpha=&FloatFromCurve(FloatCurve=DataInterface() {propertyValues {Curve {keys [<at 0 0.5>,<at 0.5 0.5>,<at 1 0>]}}}, Scale Curve=2)
    Scale RGB=&MakeCustomVectorFromBool(
     Bool=$User.bIsPainted
     True Vector=&VectorFromFloat(Value=&FloatFromCurve(FloatCurve=DataInterface() {propertyValues {Curve {keys [<at 0.1 20>,<at 0.15 8.5>,<at 0.8 5>,<at 1 1>]}}}))
     False Vector=&Multiply_VectorByFloat(
      Vector=&MakeVectorFromLinearColorRGB(Color=&ColorFromCurve(ColorCurve=DataInterface() {propertyValues {
          AlphaCurve {keys [<at 0 1 RCIM_Cubic>,<at 1 1 RCIM_Cubic>]}
          BlueCurve {keys [<at 0.0662 0.6 RCIM_Cubic>,<at 0.2338 0.35 RCIM_Cubic>]}
          GreenCurve {keys [<at 0.0662 0.15 RCIM_Cubic>,<at 0.2338 0.07 RCIM_Cubic>]}
          RedCurve {keys [<at 0.0662 3 RCIM_Cubic>,<at 0.2338 0.5 RCIM_Cubic>]}
         }}))
      Float=5
     )
    )
    ScaleA=true
    ScaleRGB=true
   ) {enabled=false}
   &ScaleRibbonWidth(Ribbon Width Scale=&FloatFromCurve(FloatCurve=DataInterface() {propertyValues {Curve {keys [
         {interpMode=RCIM_Cubic, tangentMode=RCTM_User, time=0, value=0}
         {arriveTangentWeight=0.0692, interpMode=RCIM_Cubic, leaveTangentWeight=0.2768, tangentMode=RCTM_User, time=0.1, value=1.4}
         {interpMode=RCIM_Cubic, tangentMode=RCTM_User, time=1, value=0}
        ]}}}))
   &DynamicMaterialParameters(
    Index 0 Param 1=2
    Index 0 Param 2=&MakeCustomFloatFromBool(Bool=$User.bIsPainted, True Float=0.2, False Float=0.5)
    Module.Index0Float0Write=true
    Module.Index0Float1Write=true
    Module.Index0Float2Write=true
    Module.Index0Float3Write=true
   )
   /SetVariables(
    Particles.R1 Scaled Color:LinearColor=&VC_DynamicScaleColor(
     Color to Scale:LinearColor=$Particles.R1 Color
     RGB Scale:Vector3f=&MakeCustomPositionFromBool(
      Bool:NiagaraBool=$User.bIsPainted
      True Position:NiagaraPosition=&VectorFromFloat(Value=&FloatFromCurve(FloatCurve=DataInterface() {propertyValues {Curve {
           defaultValue=1
           keys [
            <at 0 0 RCIM_Cubic>
            {arriveTangentWeight=0.3621, interpMode=RCIM_Cubic, leaveTangentWeight=0.5431, tangentMode=RCTM_User, time=0.3, value=1}
            {arriveTangent=-0.0405, interpMode=RCIM_Cubic, leaveTangent=-0.0405, tangentMode=RCTM_User, time=0.9, value=0.1}
           ]
          }}}))
      False Position:NiagaraPosition=&VectorFromFloat(Value=&FloatFromCurve(FloatCurve=DataInterface() {propertyValues {Curve {
           defaultValue=1
           keys [
            {arriveTangent=-0.0529, interpMode=RCIM_Cubic, leaveTangent=-0.0529, tangentMode=RCTM_User, time=0, value=0}
            {arriveTangentWeight=0.3621, interpMode=RCIM_Cubic, leaveTangentWeight=0.5431, tangentMode=RCTM_User, time=0.255, value=0.5}
            {arriveTangent=-0.0405, interpMode=RCIM_Cubic, leaveTangent=-0.0405, tangentMode=RCTM_User, time=0.9, value=0.1}
           ]
          }}}))
     )
     Alpha Scale:NiagaraFloat=&FloatFromCurve(FloatCurve=DataInterface() {propertyValues {Curve {keys [{tangentMode=RCTM_User, time=0, value=0.2}, {tangentMode=RCTM_User, time=0.5, value=1}, <at 0.85 0>]}}})
    )
    Particles.R1 Scaled Width:NiagaraFloat=&Multiply_Float(
     A=$Particles.R1 Width
     B=&FloatFromCurve(FloatCurve=DataInterface() {propertyValues {Curve {
         defaultValue=0
         keys [
          {arriveTangent=-0.0088, interpMode=RCIM_Cubic, leaveTangent=-0.0088, tangentMode=RCTM_User, time=0, value=0.8}
          {arriveTangent=-0.2816, interpMode=RCIM_Cubic, leaveTangent=-0.2816, tangentMode=RCTM_User, time=1, value=0}
         ]
        }}})
    )
   ) {isSetParameters=true}
   &ScaleColor(
    Scale Alpha=&FloatFromCurve(FloatCurve=DataInterface() {propertyValues {Curve {
        defaultValue=1
        keys [
         {arriveTangent=4.1785, interpMode=RCIM_Cubic, leaveTangent=4.1785, tangentMode=RCTM_User, time=0, value=0}
         {arriveTangentWeight=0.1021, interpMode=RCIM_Cubic, leaveTangentWeight=0.3979, tangentMode=RCTM_User, time=0.1, value=1}
         {arriveTangent=0.0224, interpMode=RCIM_Cubic, leaveTangent=0.0224, tangentMode=RCTM_User, time=1, value=1}
        ]
       }}})
    ScaleA=true
    ScaleRGB=true
   ) {enabled=false, module=ScaleColor001}
  ]
  rendererClasses [&NiagaraRibbonRendererProperties]
  renderers [
   {rendererClass=&NiagaraRibbonRendererProperties}
   {rendererClass=&NiagaraRibbonRendererProperties, rendererIndex=1}
   {rendererClass=&NiagaraRibbonRendererProperties, rendererIndex=2}
  ]
 }
}
summary {
 systemName=NS_Boost_Psydrill
 emitters [
  {
   emitterName=NE_Psydrill_Cone001
   rendererClasses [&NiagaraMeshRendererProperties, &NiagaraSpriteRendererProperties]
  }
  {emitterName=NE_Psydrill_Ribbons001, rendererClasses [&NiagaraRibbonRendererProperties]}
  {
   emitterName=NE_Psydrill_Particles001
   rendererClasses [&NiagaraSpriteRendererProperties, &NiagaraMeshRendererProperties]
  }
 ]
 userVariables [
  User.bIsFrontendPreview {type=NiagaraBool}
  User.bIsPainted {type=NiagaraBool}
  User.ColorVariantID {type=NiagaraInt32}
  User.Drill {type=LinearColor, description="", defaultValue {struct=LinearColor, value=<rgba 1 1 1 1>}}
  User.DTCarOffset {type=Vector3f, description="", defaultValue {struct=Vector3f, value=<0 0 0>}}
  User.GameModeID {type=NiagaraInt32}
  User.Impact {type=LinearColor, description="", defaultValue {struct=LinearColor, value=<rgba 1 1 1 1>}}
  User.PaintedColor {type=LinearColor, description="", defaultValue {struct=LinearColor, value=<rgba 1 1 1 1>}}
  User.PaintedScalar {type=NiagaraFloat, defaultValue=1, description=""}
  User.PortCount {type=NiagaraInt32, defaultValue=1}
  User.SingleBoostPos {type=NiagaraPosition, description="", defaultValue {}}
 ]
}
